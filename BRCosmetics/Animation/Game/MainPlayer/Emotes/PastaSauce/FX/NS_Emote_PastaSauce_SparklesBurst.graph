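# Niagara sparkle effect for the PastaSauce player emote, built from three twinkling sprite emitters that use waveform-driven alpha.
# SparkleRainCPULow drops CPU bursts from above the system, gated by system bar and beat modulo values, with curl-noise and vector-field drift.
# GPU emitter Tinkerbell002 spawns sparkles on the sampled skeletal mesh.
# SparkleRainTrail, also GPU, spawns trails from Tinkerbell002's particles via an attribute reader, despite its name.
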NiagaraSystem /BRCosmetics/Animation/Game/MainPlayer/Emotes/PastaSauce/FX/NS_Emote_PastaSauce_SparklesBurst
dependencies {
 usedDataInterfaces [
  &NiagaraDataInterfaceParticleRead
  &NiagaraDataInterfaceCurve
  /Script/Niagara.NiagaraDataInterfaceVectorField
  &NiagaraDataInterfaceSkeletalMesh
 ]
 usedDynamicInputs [
  &Add_Integer
  /Niagara/DynamicInputs/Divide/Divide_Int
  &ModuloInt
  &SetBoolByFloatComparison
  &MakeFloatFromVector
 ]
 usedModules [
  &self:SystemSpawnScript.SystemScriptSource.SystemScriptGraph.NiagaraNodeAssignment_3.SetVariables_#1
  &SystemState
  &SampleSkeletalMesh
  &self:SystemSpawnScript.SystemScriptSource.SystemScriptGraph.NiagaraNodeAssignment_0.SetVariables_#2
  &self:SystemSpawnScript.SystemScriptSource.SystemScriptGraph.NiagaraNodeAssignment_1.SetVariables_#3
  &self:SystemSpawnScript.SystemScriptSource.SystemScriptGraph.NiagaraNodeAssignment_2.SetVariables_#4
  &self:SystemSpawnScript.SystemScriptSource.SystemScriptGraph.NiagaraNodeAssignment_5.SetVariables_#5
 ]
}
emitters {
 SparkleRainCPULow {
  emitterUpdateScript [
   &EmitterState(Life Cycle Mode=::NewEnumerator1, Loop Duration=$System.CurrentLoopDuration, UseLoopDelay=false)
   &SpawnBurst_Instantaneous(
    Spawn Count=22
    Spawn Probability=&MakeCustomFloatFromBool(
     Bool=&Boolean_AndOperation(
      Bool1=&SetBoolByIntComparison(A=$System.BarModulo, B=1, Comparison Type=::NewEnumerator4)
      Bool2=&SetBoolByIntComparison(A=$System.BeatModulo, B=1, Comparison Type=::NewEnumerator4)
     )
     True Float=1
     False Float=0
    )
    Use Spawn Probability=true
    UseLoopCountLimit=false
   )
  ]
  particleSpawnScript [
   &InitializeParticle(
    AdjustAlpha=false
    AdjustHue=true
    AdjustSaturation=false
    AdjustValue=true
    Color=<rgba 0 0.6667 1 1>
    Color Mode=::NewEnumerator3
    Hue Shift Range:Vector2f=<-0.05 0.05>
    Lifetime Max=1.666
    Lifetime Min=0.666
    Lifetime Mode=::NewEnumerator1
    Sprite Rotation Angle=0.125
    Sprite Rotation Mode=::NewEnumerator1
    Sprite Size Mode=::NewEnumerator1
    Uniform Sprite Size Max=33
    Uniform Sprite Size Min=16
    UsePositionOffset=false
    Value Range:Vector2f=<0.3 11>
    Write Material Random=true
   )
   &AddVelocity(
    Cone Axis=<0 0 -1>
    Module.UseConeFalloff=true
    Module.UseVelDistribution=true
    Override Seed=false
    Velocity Mode=::NewEnumerator2
    Velocity Speed=&RandomRangeFloat(Minimum=111, Maximum=333)
   )
   &SystemLocation(MaskSpawnGroup=false, Offset=<0 0 280>)
  ]
  particleUpdateScript [
   &GravityForce(Gravity=<0 0 -64>)
   &Drag(Use Linear Drag=true, Use Rotational Drag=true)
   &CurlNoiseForce(Noise Frequency=0.25, Noise Strength=15, Pan Noise=false)
   &SolveForcesAndVelocity(Clamp Velocity=false, Limit Acceleration=false)
   &ParticleState()
   &DynamicMaterialParameters(
    Index 0 Param 1=0.8
    Index 0 Param 2=1
    Index 0 Param 3=1
    Index 1 Param 1=150
    Index 1 Param 4=1
    Module.Param0WriteEnabled=true
    Module.Param0WriteEnabled1=true
    Module.Param1WriteEnabled=true
    Module.Param1WriteEnabled1=true
    Module.Param2WriteEnabled=true
    Module.Param2WriteEnabled1=true
    Module.Param3WriteEnabled=false
    Module.Param3WriteEnabled1=true
    Write Parameter Index 1=true
   )
   &ScaleColor(
    Scale Alpha=&FloatFromCurve(
     FloatCurve=DataInterface() {propertyValues {Curve {keys [{interpMode=RCIM_Cubic, tangentMode=RCTM_User, time=0, value=1}, {interpMode=RCIM_Cubic, tangentMode=RCTM_User, time=1, value=0}]}}}
     Scale Curve=&Waveform(
      Global Phase=&Add_Float(A=$Particles.Age, B=&RandomRangeFloat(Minimum=0, Maximum=1))
      "Waveform [1]"=::NewEnumerator0
      "[1] Frequency"=&RandomRangeFloat(Minimum=1.618, Maximum=3.618)
      "[1] Amplitude Min/Max"=<0 1>
     )
    )
    Scale RGB=&VectorFromFloat(Value=&Lerp_Float(
      A=1
      B=3
      Alpha=&FloatFromCurve(FloatCurve=DataInterface() {propertyValues {Curve {keys [<at 0 1>,<at 0.25 0>]}}})
     ))
    ScaleA=true
    ScaleRGB=true
   )
   &ScaleSpriteSize(Uniform Curve Sprite Scale=DataInterface() {propertyValues {Curve {keys [
        {arriveTangent=2.9868, time=0, value=0}
        {leaveTangent=0.2875, time=0.1, value=0.833}
        <at 0.9 1>
        {arriveTangent=-0.0405, leaveTangent=-0.0405, time=1, value=0}
       ]}}})
   /Niagara/VectorFields/SampleVectorField(
    FieldRotate:Quat4f=<0 0 1 1>
    Sampled Vector Scale=0.05
    VectorField=DataInterface() {propertyValues {Field=/Game/Effects/VectorFields/Noisy_Turbulence_103, bTileX=true, bTileY=true, bTileZ=true}}
   )
   /Niagara/VectorFields/ApplyVectorField(Vector Field Intensity as Force=0, Vector Field Intensity as Velocity=50, Vector Field Velocity Tightness=0.001)
  ]
  rendererClasses [&NiagaraSpriteRendererProperties]
  renderers [{rendererClass=&NiagaraSpriteRendererProperties}]
 }
 SparkleRainTrail {
  simTarget=GPUComputeSim
  emitterUpdateScript [
   &EmitterState(Life Cycle Mode=::NewEnumerator1, Loop Behavior=::NewEnumerator1, Loop Duration=2, UseLoopDelay=false)
   &SpawnParticlesFromOtherEmitter(
    Attribute Reader=DataInterface() {propertyValues {EmitterBinding {bindingMode=Other, emitterName=Tinkerbell002}}}
    Maximum Age Spawn Threshold=2
    Minimum Age Spawn Threshold=0
    Spawn Rate=6
    Spawn Rate Per Particle Cap=22
    Use Maximum Age=false
    Use Minimum Age=false
   )
  ]
  particleSpawnScript [
   &InitializeParticle(
    Color=&RandomRangeLinearColor(Minimum=<rgba 0.4886 0.5173 1 1>, Maximum=<rgba 0.1968 0.249 1 1>)
    Lifetime Max=1
    Lifetime Min=0.333
    Lifetime Mode=::NewEnumerator1
    Sprite Rotation Angle=0.125
    Sprite Rotation Mode=::NewEnumerator1
    Sprite Size Mode=::NewEnumerator1
    Uniform Sprite Size Max=11
    Uniform Sprite Size Min=11
    UsePositionOffset=false
    Write Material Random=true
   )
   &SampleParticlesFromOtherEmitter(
    Applied Velocity Scale:NiagaraFloat=&FloatFromCurve(
     FloatCurve=DataInterface() {propertyValues {Curve {keys [{interpMode=RCIM_Cubic, tangentMode=RCTM_Break, time=0, value=1}, <at 1 0 RCIM_Cubic>]}}}
     Scale Curve=0.5
    )
    Maximum Sampled Particle Age=2
    Minimum Sampled Particle Age=0
    Module.MaxSampledAgeCutoff=true
    Module.MinSampledAgeCutoff=true
    Reject Particle By Sampled Age=true
    Velocity Sampling=::NewEnumerator0
   )
   &ShapeLocation(MaskSpawnGroup=false, Override Seed=false, Sphere Radius=10, UseOwnerScale=true)
  ]
  particleUpdateScript [
   &ParticleState()
   &Drag(Drag=2, Use Linear Drag=true, Use Rotational Drag=false)
   &GravityForce(Gravity=&RandomRangeVector(Minimum=<0 0 -50>, Maximum=<1 1 -100>))
   &CurlNoiseForce(Noise Frequency=1, Noise Strength=15, Pan Noise=false)
   &SolveForcesAndVelocity(Clamp Velocity=false, Limit Acceleration=false)
   &ScaleColor(
    Scale Alpha=&FloatFromCurve(
     FloatCurve=DataInterface() {propertyValues {Curve {keys [
         {tangentMode=RCTM_User, time=0, value=1}
         {arriveTangent=-0.0405, interpMode=RCIM_Cubic, leaveTangent=-0.0405, tangentMode=RCTM_User, time=0.7009, value=1}
         {arriveTangent=-8.0089, interpMode=RCIM_Cubic, leaveTangent=-8.0089, tangentMode=RCTM_User, time=1, value=0}
        ]}}}
     Scale Curve=&Waveform(
      Global Phase=&Add_Float(A=$Particles.Age, B=&RandomRangeFloat(Minimum=0, Maximum=1))
      "Waveform [1]"=::NewEnumerator0
      "[1] Frequency"=&RandomRangeFloat(Minimum=3.618, Maximum=11.08)
      "[1] Amplitude Min/Max"=<0 1>
     )
    )
    Scale RGB=&VectorFromFloat(Value=&Lerp_Float(
      A=0.25
      B=22
      Alpha=&FloatFromCurve(FloatCurve=DataInterface() {propertyValues {Curve {keys [<at 0 1>,<at 0.333 0>]}}})
     ))
    ScaleA=true
    ScaleRGB=true
   )
   &ScaleSpriteSize(Uniform Curve Sprite Scale=DataInterface() {propertyValues {Curve {keys [
        {interpMode=RCIM_Cubic, tangentMode=RCTM_User, time=0, value=0}
        {interpMode=RCIM_Cubic, tangentMode=RCTM_User, time=0.5, value=1}
        {tangentMode=RCTM_User, time=1, value=0}
       ]}}})
  ]
  rendererClasses [&NiagaraSpriteRendererProperties]
  renderers [{rendererClass=&NiagaraSpriteRendererProperties}]
 }
 Tinkerbell002 {
  simTarget=GPUComputeSim
  emitterUpdateScript [
   &EmitterState(Life Cycle Mode=::NewEnumerator1, Loop Behavior=::NewEnumerator1, Loop Delay=7, Loop Duration=1.5, UseLoopDelay=false)
   &SpawnRate(SpawnRate=20, Use Spawn Probability=false)
   &SpawnBurst_Instantaneous(Spawn Count=5, Use Spawn Probability=false, UseLoopCountLimit=false)
  ]
  particleSpawnScript [
   &PartitionParticles() {enabled=false}
   &InitializeParticle(
    Color=<rgba 0.5349 0.5662 1 0.49>
    Position Offset=<0 0 100>
    Sprite Rotation Mode=::NewEnumerator2
    Sprite Size Mode=::NewEnumerator3
    Uniform Sprite Size=20
    UsePositionOffset=true
    Write Material Random=true
   )
   &SkeletalMeshLocation(
    Bone Sampling Mode=::NewEnumerator0
    "Inherit Velocity (Scale)"=0.5
    MaskSpawnGroup=false
    OverrideRandomSeed=false
    Skeletal Mesh=$System.Skeletal Mesh
    Velocity Sampling=::NewEnumerator0
    VelocitySpeedLimit=false
   )
   &ShapeLocation(
    MaskSpawnGroup=false
    Override Seed=false
    Shape Origin=$Output.SkeletalMeshLocation.SampledPositionWithoutOffset
    Sphere Radius=15
    UseOwnerScale=true
   ) {enabled=false}
   &AddVelocity(
    Cone Angle=60
    Cone Axis=<0 0 1>
    Module.UseConeFalloff=true
    Module.UseVelDistribution=true
    Override Seed=false
    Velocity Mode=::NewEnumerator2
    Velocity Speed=&RandomRangeFloat(Minimum=50, Maximum=100)
   ) {enabled=false}
   &InheritVelocity(Inherited Velocity Amount Scale=&VectorFromFloat(Value=2), Speed Limit=true, UseSpeedThreshold=false) {enabled=false}
  ]
  particleUpdateScript [
   &SampleVectorField(
    FieldRotate:Quat4f=<0 0 1 1>
    Sampled Vector Scale=0.05
    VectorField=DataInterface() {propertyValues {Field=&Noisy_Turbulence_103, bTileX=true, bTileY=true, bTileZ=true}}
   ) {enabled=false}
   &ParticleState()
   &SpriteRotationRate(Rotation Rate=&MaskFloatBySpawnGroup(Masked Float=45, Passthrough Float=-45))
   &ScaleSpriteSize(Uniform Curve Sprite Scale=DataInterface() {propertyValues {Curve {keys [
        {arriveTangent=9.0725, interpMode=RCIM_Cubic, leaveTangent=9.0725, tangentMode=RCTM_User, time=0, value=0}
        {arriveTangent=-0.0039, interpMode=RCIM_Cubic, leaveTangent=-0.0039, tangentMode=RCTM_User, time=0.1408, value=0.9847}
        {arriveTangent=-4.4862, interpMode=RCIM_Cubic, leaveTangent=-4.4862, tangentMode=RCTM_User, time=1, value=0}
       ]}}})
   &ScaleColor(
    Scale Alpha=&FloatFromCurve(
     FloatCurve=DataInterface() {propertyValues {Curve {keys [{interpMode=RCIM_Cubic, tangentMode=RCTM_User, time=0, value=1}, {interpMode=RCIM_Cubic, tangentMode=RCTM_User, time=1, value=0}]}}}
     Scale Curve=&Waveform(
      Global Phase=&Add_Float(A=$Particles.Age, B=&RandomRangeFloat(Minimum=0, Maximum=1))
      "Waveform [1]"=::NewEnumerator0
      "[1] Frequency"=&RandomRangeFloat(Minimum=1.618, Maximum=3.618)
      "[1] Amplitude Min/Max"=<0 1>
     )
    )
    Scale RGB=&VectorFromFloat(Value=&Lerp_Float(
      A=1
      B=3
      Alpha=&FloatFromCurve(FloatCurve=DataInterface() {propertyValues {Curve {keys [<at 0 1>,<at 0.25 0>]}}})
     ))
    ScaleA=true
    ScaleRGB=true
   )
   &DynamicMaterialParameters(
    Index 0 Param 1=1
    Index 0 Param 2=1
    Index 0 Param 3=1
    Module.Param0WriteEnabled=true
    Module.Param1WriteEnabled=true
    Module.Param2WriteEnabled=true
    Module.Param3WriteEnabled=true
   )
   &InheritSourceMovement(
    Applied Force Scale=&VectorFromFloat(Value=1)
    Applied Position Scale=&VectorFromFloat(Value=1)
    ApplyForce=true
    ApplyForceFalloff=false
    ApplyPosition=false
    Module.LimitForceSpeed=false
   ) {enabled=false}
   &Drag(Drag=&RandomRangeFloat(Minimum=2, Maximum=3), Use Linear Drag=true, Use Rotational Drag=true)
   &CurlNoiseForce(Noise Frequency=0.25, Noise Strength=15, Pan Noise=false) {enabled=false}
   &GravityForce(Gravity=&RandomRangeVector(Minimum=<0 0 -25>, Maximum=<0 0 -50>))
   &SolveForcesAndVelocity(Clamp Velocity=false, Limit Acceleration=false)
   &ApplyVectorField(Vector Field Intensity as Force=0, Vector Field Intensity as Velocity=20, Vector Field Velocity Tightness=0.001) {enabled=false}
  ]
  rendererClasses [&NiagaraSpriteRendererProperties]
  renderers [{rendererClass=&NiagaraSpriteRendererProperties}]
 }
}
summary {
 systemName=NS_Emote_PastaSauce_SparklesBurst
 emitters [
  {emitterName=SparkleRainTrail, simTarget=GPUComputeSim, rendererClasses [&NiagaraSpriteRendererProperties]}
  {emitterName=SparkleRainCPULow, rendererClasses [&NiagaraSpriteRendererProperties]}
  {emitterName=Tinkerbell002, simTarget=GPUComputeSim, rendererClasses [&NiagaraSpriteRendererProperties]}
 ]
 userVariables [
  User.bIsFrontEnd {type=NiagaraBool}
  User.bIsFrontEndPreview {type=NiagaraBool}
  User.ConfettiColorArray {
   type=NiagaraDataInterfaceArrayColor
   description=""
   defaultValue {
    struct=NiagaraExt_VariableValue_DataInterface
    value {
     dataInterface=&self:NiagaraDataInterfaceArrayColor_1
     dataInterfaceClass=&NiagaraDataInterfaceArrayColor
    }
   }
  }
  User.FeetRestingPositions {
   type=NiagaraDataInterfaceArrayFloat3
   description=""
   defaultValue {
    struct=NiagaraExt_VariableValue_DataInterface
    value {
     dataInterface=&self:NiagaraDataInterfaceArrayFloat3_2
     dataInterfaceClass=&NiagaraDataInterfaceArrayFloat3
    }
   }
  }
  User.Local Alpha Preview {type=NiagaraFloat, defaultValue=0.75, description=""}
  User.Other Alpha Preview {type=NiagaraFloat, defaultValue=0.25, description=""}
  User.Ranked_Curve {type=NiagaraFloat, defaultValue=0, description=""}
  User.SongBPM {type=NiagaraFloat, defaultValue=97.5, description=""}
  User.SpotlightColorArray {
   type=NiagaraDataInterfaceArrayColor
   description=""
   defaultValue {
    struct=NiagaraExt_VariableValue_DataInterface
    value {
     dataInterface=&self:NiagaraDataInterfaceArrayColor_0
     dataInterfaceClass=&NiagaraDataInterfaceArrayColor
    }
   }
  }
  User.SpotlightsOrientationArray {
   type=NiagaraDataInterfaceArrayFloat3
   description=""
   defaultValue {
    struct=NiagaraExt_VariableValue_DataInterface
    value {
     dataInterface=&self:NiagaraDataInterfaceArrayFloat3_1
     dataInterfaceClass=&NiagaraDataInterfaceArrayFloat3
    }
   }
  }
  User.SpotlightsPositionArray {
   type=NiagaraDataInterfaceArrayFloat3
   description=""
   defaultValue {
    struct=NiagaraExt_VariableValue_DataInterface
    value {
     dataInterface=&self:NiagaraDataInterfaceArrayFloat3_0
     dataInterfaceClass=&NiagaraDataInterfaceArrayFloat3
    }
   }
  }
 ]
}
